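# Example stereo MetaSound Source demonstrating the Polyphony Wave Player node from the MetasoundPolyphonyInternal plugin.
# A Trigger Repeat, started on play and timed by Period, repeatedly fires the player with a three-entry subset of a sound wave container selected by Index.
# The player feeds left and right outputs and logs voice start and finish events.
# Other declared inputs, including Play, Stop, Seed and Pitch Shift, are left unconnected.
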MetaSoundSource /MetasoundPolyphonyInternal/Examples/WavePlayer/MSS_PolyphonyWavePlayer
inputs [
 Avoid Repeating Last {type=Int32}
 Index {type=Int32}
 Period {type=Time}
 Pitch Shift {type=Float}
 Play {type=Trigger}
 Reset {type=Trigger}
 Seed {type=Int32}
 Sound Wave Container {type=Polyphony::SoundWaveContainerAsset}
 Stop {type=Trigger}
 UE.Source.OnPlay {type=Trigger}
]
interfaces [UE.OutputFormat.Stereo:1.0, UE.Source:1.1]
outputs [
 "UE.OutputFormat.Stereo.Audio:0" {type=Audio}
 "UE.OutputFormat.Stereo.Audio:1" {type=Audio}
]
pages [Default {
  input_defaults {
   Avoid Repeating Last=0
   Index=5
   Period=1
   Pitch Shift=0
   Play=null
   Reset=null
   Seed=12345
   Sound Wave Container=null
   Stop=null
   UE.Source.OnPlay=null
  }
  nodes [
   InitVariable.Polyphony::SoundWaveContainerAsset()
   PrintLogPrintLog_0 = Print Log.Print Log:String() {also {defaults {Label="", Value To Log=Looped}}}
   PrintLogPrintLog_1 = Print Log.Print Log:String() {also {defaults {Label="", Value To Log=Finished}}}
   MetasoundPolyphonyInternalPolyphonyWavePlayer_0 = MetasoundPolyphonyInternal.PolyphonyWavePlayer(
    Sound Wave Container=MetasoundPolyphonyInternal.Subset:SoundWaveContainer(Index=^Index) {also {defaults {Num=3, Sound Wave Container=/MetasoundPolyphonyInternal/Examples/WavePlayer/Sounds/SoundWaveContainer}, version [0,1]}}.Sound Wave Container
    Play=UE.TriggerRepeat(Period=^Period, Start=^UE.Source.OnPlay).RepeatOut
   ) {also {version [0,1]}}
   PrintLogPrintLog_2 = Print Log.Print Log:String(Trigger=@MetasoundPolyphonyInternalPolyphonyWavePlayer_0.On Voices Started) {also {defaults {Value To Log=Voices Started}}}
   PrintLogPrintLog_3 = Print Log.Print Log:String(Trigger=@MetasoundPolyphonyInternalPolyphonyWavePlayer_0.On Voices Finished) {also {defaults {Label="", Value To Log=Voices Finished}}}
  ]
  links [
   {dst="@UE.OutputFormat.Stereo.Audio:0", src=MetasoundPolyphonyInternalPolyphonyWavePlayer_0.Out Left}
   {dst="@UE.OutputFormat.Stereo.Audio:1", src=MetasoundPolyphonyInternalPolyphonyWavePlayer_0.Out Right}
  ]
 }]
properties {
 assetImportData=&self:AssetImportData
 bLooping=true
 duration=10000
 numChannels=2
 outputFormat=Stereo
 sampleRate=48000
 volume=0.25
}
