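; Example actor Blueprint that continuously modulates a MetaSound "Period" float parameter at runtime.
; BeginPlay creates a generator handle from the Audio variable and makes a MetaSound parameter pack, storing both.
; Tick accumulates time, maps its sine to an oscillation between 0.5 and 1.5, writes it into the pack and applies the pack to the generator handle on success.
; The overlap event and construction script are empty.
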(event EventBeginPlay
  (bind _returnvalue (MetaSound|CreateMetaSoundGeneratorHandle (Variables|Default|GetAudio)))
  (Variables|Default|SetGeneratorHandle _returnvalue)
  (bind _returnvalue_2 (MetaSoundParameterPack|MakeMetasoundParameterPack))
  (Variables|Default|SetPack _returnvalue_2))

(event Collision|EventActorBeginOverlap
  (params (OtherActor :type Actor%20Object%20Reference)))

(event EventTick
  (params (DeltaSeconds :type Float%20%28single-precision%29))
  (bind _output_get (Variables|Default|SetTime (+ DeltaSeconds (Variables|Default|GetTime))))
  (MetaSoundParameterPack|SetFloat (Variables|Default|GetPack) "Period" (/ (+ (Math|Trig|Sin(Radians) _output_get) 2.0) 2.0) false
    (:Succeeded
      (MetaSoundParameterPack|ApplyParameterPack (Variables|Default|GetGeneratorHandle) (Variables|Default|GetPack)))))

(fn ConstructionScript)
